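# Sound cue that plays a short musical one-shot stinger for the Town Square point of interest in Wild Estate.
# A Random node picks one of five equally weighted stinger waves and feeds a Modulator that keeps pitch fixed and can vary volume.
# It routes to the music sound class with a musical one-shot attenuation, and overrides concurrency so only one instance plays, blocking new ones.
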SoundCue /WildEstate/Assets/Audio/Music/MusicOneshots/S38_MusicalOneShots_WildEstate_TownSquare
root=Modulator_0
nodes [
 {id=Modulator_0, children [Random_0], properties {pitchMax=1, pitchMin=1, volumeMax=1}}
 {id=Random_0, children [WavePlayer_0,WavePlayer_1,WavePlayer_2,WavePlayer_3,WavePlayer_4], properties {weights [1,1,1,1,1]}}
 {id=WavePlayer_0, properties {soundWaveAssetPtr=&/MX_S38_POI_IP_TS_OneShot_5_STNG}}
 {id=WavePlayer_1, properties {soundWaveAssetPtr=&/MX_S38_POI_IP_TS_OneShot_4_STNG}}
 {id=WavePlayer_2, properties {soundWaveAssetPtr=&/MX_S38_POI_IP_TS_OneShot_3_STNG}}
 {id=WavePlayer_3, properties {soundWaveAssetPtr=&/MX_S38_POI_IP_TS_OneShot_2_STNG}}
 {id=WavePlayer_4, properties {soundWaveAssetPtr=&/MX_S38_POI_IP_TS_OneShot_1_STNG}}
]
properties {
 attenuationSettings=/AmbientAudioBase/Attenuations/Fort_Ambient_Athena_Musical_Oneshot_NoVerb_LowEQ_Att
 bOverrideConcurrency=true
 maxDistance=2097152
 soundClassObject=/Game/Sounds/FortSoundClassesAndModes/Fort_Music
 volumeMultiplier=0.9
 concurrencyOverrides {maxCount=1, resolutionRule=PreventNew, voiceStealReleaseTime=2}
}
